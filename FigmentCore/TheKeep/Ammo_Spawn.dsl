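; Ammo_Spawn is a searchable toolbox-style ammo pickup in The Keep, spawned by a BKeepAmmoStash.
; InitializePickup stores the stash reference, slot position and tier, then initializes the container; GetLootTierOverride returns that tier.
; On loot it reports AmmoTaken to the stash with its position and destroys itself.
; Search, repeat-loot, damage and death events play toolbox foley and impact sounds at the actor's location.
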(event EventOnLoot
  (Class|BKeepAmmoStash|AmmoTaken (Variables|Default|GetSpawnerRef) (Variables|Default|GetPositionRef) self)
  (Actor|DestroyActor))

(event EventOnBeginSearch
  (Audio|PlaySoundatLocation "/Game/Sounds/Foley_Loot/Containers/Toolbox/Toolbox_Searching_Long_Loop_Cue.Toolbox_Searching_Long_Loop_Cue" (Transformation|GetActorLocation)))

(event EventOnLootRepeat
  (Audio|PlaySoundatLocation "/Game/Sounds/Foley_Loot/Containers/Toolbox/Toolbox_Searching_End_Cue.Toolbox_Searching_End_Cue" (Transformation|GetActorLocation)))

(event EventOnDamagePlayEffects
  (params (Damage :type Float%20%28single-precision%29) (DamageTags :type Gameplay%20Tag%20Container%20Structure%20%28by%20ref%29) (Momentum :type Vector :default "0, 0, 0") (HitInfo :type Hit%20Result%20Structure%20%28by%20ref%29) (InstigatedBy :type Fort%20Pawn%20Object%20Reference) (DamageCauser :type Actor%20Object%20Reference) (EffectContext :type Gameplay%20Effect%20Context%20Handle%20Structure))
  (Audio|PlaySoundatLocation "/Game/Sounds/Fort_Impact_Sounds/Demolition/Toolbox/Demo_Toolbox_Impact_Cue.Demo_Toolbox_Impact_Cue" (Transformation|GetActorLocation)))

(event EventOnDeathPlayEffects
  (params (Damage :type Float%20%28single-precision%29) (DamageTags :type Gameplay%20Tag%20Container%20Structure%20%28by%20ref%29) (Momentum :type Vector :default "0, 0, 0") (HitInfo :type Hit%20Result%20Structure%20%28by%20ref%29) (InstigatedBy :type Fort%20Pawn%20Object%20Reference) (DamageCauser :type Actor%20Object%20Reference) (EffectContext :type Gameplay%20Effect%20Context%20Handle%20Structure))
  (Audio|PlaySoundatLocation "/Game/Sounds/Fort_Impact_Sounds/Demolition/Toolbox/Demo_Toolbox_Impact_Cue.Demo_Toolbox_Impact_Cue" (Transformation|GetActorLocation)))

(fn GetLootTierOverride
  (return (Variables|Default|GetPickupTier)))

(fn InitializePickup
  (params (Parent :type B%20Keep%20Ammo%20Stash%20Object%20Reference) (Position :type Integer) (Tier :type Integer))
  (Variables|Default|SetSpawnerRef Parent)
  (Variables|Default|SetPositionRef Position)
  (Variables|Default|SetPickupTier Tier)
  (Utilities|Casting|CastToFortPlayerController (Game|GetPlayerController 0))
  (Keep|InitializeContainer 0))

(fn ConstructionScript)
