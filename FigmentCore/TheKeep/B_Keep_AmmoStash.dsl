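; Keep ammo stash that spawns Ammo_Spawn pickups at up to eight placement points for mission defense.
; BeginPlay and defense start read pickup tier, plus cap and cooldown by player count, from the KeepAmmoStash table.
; A gated delay loop fills free slots quickly during warmup, then per cooldown; AmmoTaken frees slots.
; Defense stop destroys pickups and resets the container through player controller 0, marked as a temporary solution.
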(event EventBeginPlay
  (bind _self self)
  (bind _ammopositions (Variables|Default|GetAmmoPositions))
  (bind _pickuptier (Variables|Default|GetPickupTier))
  (Utilities|FlowControl|Sequence
    (:then_0
      (Utilities|FlowControl|DoOnce false :!label _label
        (:Completed
          (Variables|Default|SetPickupTier (.PickupTier (Utilities|Struct|BreakFortKeepAmmoStash (stale _returnvalue_2))))
          (if (Variables|Default|GetWarmup) :!label _label_2
            (Utilities|FlowControl|Delay 0.2)
            (else
              (Utilities|FlowControl|Delay (Math|Conversions|ToFloat(Integer) (Variables|Default|GetCooldown))))))))
    (:then_1
      _)
    (:then_2
      (Variables|Default|SetWarmup true))
    (:then_3
      (for _ (range 8)
        (Utilities|Array|Add (Variables|Default|GetAmmoPositions)))
      (bind _returnvalue (Keep|GetContainerNameandLevelKey))
      (bind _returnvalue_2 (Utilities|GetDataTableRow "/Game/Balance/DataTables/KeepAmmoStash.KeepAmmoStash" _returnvalue))
      (switch int (Utilities|Array|Length (Fortnite|GetAllFortPlayerControllers true))
        (:1
          (Variables|Default|SetMaxPickups (Utilities|Struct|BreakFortKeepAmmoStash _returnvalue_2))
          (Variables|Default|SetCooldown (.Cooldown1 (Utilities|Struct|BreakFortKeepAmmoStash _returnvalue_2))))
        (:2
          (Variables|Default|SetMaxPickups (.Max2 (Utilities|Struct|BreakFortKeepAmmoStash _returnvalue_2)))
          (Variables|Default|SetCooldown (.Cooldown2 (Utilities|Struct|BreakFortKeepAmmoStash _returnvalue_2))))
        (:3
          (Variables|Default|SetMaxPickups (.Max3 (Utilities|Struct|BreakFortKeepAmmoStash _returnvalue_2)))
          (Variables|Default|SetCooldown (.Cooldown3 (Utilities|Struct|BreakFortKeepAmmoStash _returnvalue_2))))
        (:4
          (Variables|Default|SetMaxPickups (.Max4 (Utilities|Struct|BreakFortKeepAmmoStash _returnvalue_2)))
          (Variables|Default|SetCooldown (.Cooldown4 (Utilities|Struct|BreakFortKeepAmmoStash _returnvalue_2)))))
      (goto _label)))
  (Utilities|FlowControl|Gate true
    (:Exit
      (Utilities|FlowControl|Sequence
        (:then_0
          (goto _label_2))
        (:then_1
          (if (< (Variables|Default|GetAmmoNumber) (Variables|Default|GetMaxPickups))
            (bind (_array_element _array_index) (Utilities|Array|ForEachLoopwithBreak _ammopositions :!label _label_3)
              (:LoopBody
                (if _array_element
                  (if (== _array_index 7)
                    _)
                  (else
                    (Utilities|Array|SetArrayElem _ammopositions _array_index "true")
                    (Utilities|FlowControl|Sequence
                      (:then_0
                        (goto _label_3 Break))
                      (:then_1
                        (Variables|Default|SetAmmoNumber (+ (Variables|Default|GetAmmoNumber) 1))
                        (switch int _array_index
                          (:0
                            (bind _returnvalue_3 (Game|SpawnActorfromClass "/FigmentCore/TheKeep/Ammo_Spawn.Ammo_Spawn_C" (Transformation|GetWorldTransform (Variables|Default|GetPlacement%201)) "AlwaysSpawn" "MultiplyWithRoot" 0 0 0 0))
                            (Class|AmmoSpawn|InitializePickup _returnvalue_3 _self _array_index _pickuptier)
                            (Utilities|Array|Add (Variables|Default|GetPickups) _returnvalue_3))
                          (:1
                            (bind _returnvalue_4 (Game|SpawnActorfromClass "/FigmentCore/TheKeep/Ammo_Spawn.Ammo_Spawn_C" (Transformation|GetWorldTransform (Variables|Default|GetPlacement2)) "AlwaysSpawn" "MultiplyWithRoot" 0 0 0 0))
                            (Class|AmmoSpawn|InitializePickup _returnvalue_4 _self _array_index _pickuptier)
                            (Utilities|Array|Add (Variables|Default|GetPickups) _returnvalue_4))
                          (:2
                            (bind _returnvalue_5 (Game|SpawnActorfromClass "/FigmentCore/TheKeep/Ammo_Spawn.Ammo_Spawn_C" (Transformation|GetWorldTransform (Variables|Default|GetPlacement3)) "AlwaysSpawn" "MultiplyWithRoot" 0 0 0 0))
                            (Class|AmmoSpawn|InitializePickup _returnvalue_5 _self _array_index _pickuptier)
                            (Utilities|Array|Add (Variables|Default|GetPickups) _returnvalue_5))
                          (:3
                            (bind _returnvalue_6 (Game|SpawnActorfromClass "/FigmentCore/TheKeep/Ammo_Spawn.Ammo_Spawn_C" (Transformation|GetWorldTransform (Variables|Default|GetPlacement4)) "AlwaysSpawn" "MultiplyWithRoot" 0 0 0 0))
                            (Class|AmmoSpawn|InitializePickup _returnvalue_6 _self _array_index _pickuptier)
                            (Utilities|Array|Add (Variables|Default|GetPickups) _returnvalue_6))
                          (:4
                            (bind _returnvalue_7 (Game|SpawnActorfromClass "/FigmentCore/TheKeep/Ammo_Spawn.Ammo_Spawn_C" (Transformation|GetWorldTransform (Variables|Default|GetPlacement5)) "AlwaysSpawn" "MultiplyWithRoot" 0 0 0 0))
                            (Class|AmmoSpawn|InitializePickup _returnvalue_7 _self _array_index _pickuptier)
                            (Utilities|Array|Add (Variables|Default|GetPickups) _returnvalue_7))
                          (:5
                            (bind _returnvalue_8 (Game|SpawnActorfromClass "/FigmentCore/TheKeep/Ammo_Spawn.Ammo_Spawn_C" (Transformation|GetWorldTransform (Variables|Default|GetPlacement6)) "AlwaysSpawn" "MultiplyWithRoot" 0 0 0 0))
                            (Class|AmmoSpawn|InitializePickup _returnvalue_8 _self _array_index _pickuptier)
                            (Utilities|Array|Add (Variables|Default|GetPickups) _returnvalue_8))
                          (:6
                            (bind _returnvalue_9 (Game|SpawnActorfromClass "/FigmentCore/TheKeep/Ammo_Spawn.Ammo_Spawn_C" (Transformation|GetWorldTransform (Variables|Default|GetPlacement7)) "AlwaysSpawn" "MultiplyWithRoot" 0 0 0 0))
                            (Class|AmmoSpawn|InitializePickup _returnvalue_9 _self _array_index _pickuptier)
                            (Utilities|Array|Add (Variables|Default|GetPickups) _returnvalue_9))
                          (:7
                            (bind _returnvalue_10 (Game|SpawnActorfromClass "/FigmentCore/TheKeep/Ammo_Spawn.Ammo_Spawn_C" (Transformation|GetWorldTransform (Variables|Default|GetPlacement8)) "AlwaysSpawn" "MultiplyWithRoot" 0 0 0 0))
                            (Class|AmmoSpawn|InitializePickup _returnvalue_10 _self _array_index _pickuptier)
                            (Utilities|Array|Add (Variables|Default|GetPickups) _returnvalue_10)))))))))
            (elif (Variables|Default|GetWarmup)
              (Variables|Default|SetWarmup false))))))))

(event MissionHelpers|HandleMissionEvent%28v2%29:OnDefenseStarted
  (params (InMissionGuid :type Guid%20Structure) (ObjectiveHandle :type Gameplay%20Tag%20Container%20Structure) (Params :type Fort%20Mission%20Event%20Params%20Object%20Reference) (DO_NOT_USE_THIS_OR_VARIABLES_BELOW :type Boolean) (EventFocus :type Object%20Reference) (EventContent :type Data%20Asset%20Object%20Reference) (EventInstigator :type Actor%20Object%20Reference) (GenericInt :type Integer) (GenericFloat :type Float%20%28single-precision%29) (GenericText :type Text%20%28by%20ref%29) (GameplayTags :type Gameplay%20Tag%20Container%20Structure) (MissionEvent :type Fort%20Mission%20Event%20Structure%20%28by%20ref%29))
  (bind _self self)
  (bind _ammopositions (Variables|Default|GetAmmoPositions))
  (bind _pickuptier (Variables|Default|GetPickupTier))
  (Utilities|FlowControl|Sequence
    (:then_0
      (Utilities|FlowControl|DoOnce false :!label _label
        (:Completed
          (Variables|Default|SetPickupTier (.PickupTier (Utilities|Struct|BreakFortKeepAmmoStash (stale EventBeginPlay _returnvalue_2))))
          (if (Variables|Default|GetWarmup) :!label _label_2
            (Utilities|FlowControl|Delay 0.2)
            (else
              (Utilities|FlowControl|Delay (Math|Conversions|ToFloat(Integer) (Variables|Default|GetCooldown))))))))
    (:then_1
      _)
    (:then_2
      (Variables|Default|SetWarmup true))
    (:then_3
      (for _ (range 8)
        (Utilities|Array|Add (Variables|Default|GetAmmoPositions)))
      (bind _returnvalue (Keep|GetContainerNameandLevelKey))
      (bind _returnvalue_2 (Utilities|GetDataTableRow "/Game/Balance/DataTables/KeepAmmoStash.KeepAmmoStash" _returnvalue))
      (switch int (Utilities|Array|Length (Fortnite|GetAllFortPlayerControllers true))
        (:1
          (Variables|Default|SetMaxPickups (Utilities|Struct|BreakFortKeepAmmoStash _returnvalue_2))
          (Variables|Default|SetCooldown (.Cooldown1 (Utilities|Struct|BreakFortKeepAmmoStash _returnvalue_2))))
        (:2
          (Variables|Default|SetMaxPickups (.Max2 (Utilities|Struct|BreakFortKeepAmmoStash _returnvalue_2)))
          (Variables|Default|SetCooldown (.Cooldown2 (Utilities|Struct|BreakFortKeepAmmoStash _returnvalue_2))))
        (:3
          (Variables|Default|SetMaxPickups (.Max3 (Utilities|Struct|BreakFortKeepAmmoStash _returnvalue_2)))
          (Variables|Default|SetCooldown (.Cooldown3 (Utilities|Struct|BreakFortKeepAmmoStash _returnvalue_2))))
        (:4
          (Variables|Default|SetMaxPickups (.Max4 (Utilities|Struct|BreakFortKeepAmmoStash _returnvalue_2)))
          (Variables|Default|SetCooldown (.Cooldown4 (Utilities|Struct|BreakFortKeepAmmoStash _returnvalue_2)))))
      (goto _label)))
  (Utilities|FlowControl|Gate true
    (:Exit
      (Utilities|FlowControl|Sequence
        (:then_0
          (goto _label_2))
        (:then_1
          (if (< (Variables|Default|GetAmmoNumber) (Variables|Default|GetMaxPickups))
            (bind (_array_element _array_index) (Utilities|Array|ForEachLoopwithBreak _ammopositions :!label _label_3)
              (:LoopBody
                (if _array_element
                  (if (== _array_index 7)
                    _)
                  (else
                    (Utilities|Array|SetArrayElem _ammopositions _array_index "true")
                    (Utilities|FlowControl|Sequence
                      (:then_0
                        (goto _label_3 Break))
                      (:then_1
                        (Variables|Default|SetAmmoNumber (+ (Variables|Default|GetAmmoNumber) 1))
                        (switch int _array_index
                          (:0
                            (bind _returnvalue_3 (Game|SpawnActorfromClass "/FigmentCore/TheKeep/Ammo_Spawn.Ammo_Spawn_C" (Transformation|GetWorldTransform (Variables|Default|GetPlacement%201)) "AlwaysSpawn" "MultiplyWithRoot" 0 0 0 0))
                            (Class|AmmoSpawn|InitializePickup _returnvalue_3 _self _array_index _pickuptier)
                            (Utilities|Array|Add (Variables|Default|GetPickups) _returnvalue_3))
                          (:1
                            (bind _returnvalue_4 (Game|SpawnActorfromClass "/FigmentCore/TheKeep/Ammo_Spawn.Ammo_Spawn_C" (Transformation|GetWorldTransform (Variables|Default|GetPlacement2)) "AlwaysSpawn" "MultiplyWithRoot" 0 0 0 0))
                            (Class|AmmoSpawn|InitializePickup _returnvalue_4 _self _array_index _pickuptier)
                            (Utilities|Array|Add (Variables|Default|GetPickups) _returnvalue_4))
                          (:2
                            (bind _returnvalue_5 (Game|SpawnActorfromClass "/FigmentCore/TheKeep/Ammo_Spawn.Ammo_Spawn_C" (Transformation|GetWorldTransform (Variables|Default|GetPlacement3)) "AlwaysSpawn" "MultiplyWithRoot" 0 0 0 0))
                            (Class|AmmoSpawn|InitializePickup _returnvalue_5 _self _array_index _pickuptier)
                            (Utilities|Array|Add (Variables|Default|GetPickups) _returnvalue_5))
                          (:3
                            (bind _returnvalue_6 (Game|SpawnActorfromClass "/FigmentCore/TheKeep/Ammo_Spawn.Ammo_Spawn_C" (Transformation|GetWorldTransform (Variables|Default|GetPlacement4)) "AlwaysSpawn" "MultiplyWithRoot" 0 0 0 0))
                            (Class|AmmoSpawn|InitializePickup _returnvalue_6 _self _array_index _pickuptier)
                            (Utilities|Array|Add (Variables|Default|GetPickups) _returnvalue_6))
                          (:4
                            (bind _returnvalue_7 (Game|SpawnActorfromClass "/FigmentCore/TheKeep/Ammo_Spawn.Ammo_Spawn_C" (Transformation|GetWorldTransform (Variables|Default|GetPlacement5)) "AlwaysSpawn" "MultiplyWithRoot" 0 0 0 0))
                            (Class|AmmoSpawn|InitializePickup _returnvalue_7 _self _array_index _pickuptier)
                            (Utilities|Array|Add (Variables|Default|GetPickups) _returnvalue_7))
                          (:5
                            (bind _returnvalue_8 (Game|SpawnActorfromClass "/FigmentCore/TheKeep/Ammo_Spawn.Ammo_Spawn_C" (Transformation|GetWorldTransform (Variables|Default|GetPlacement6)) "AlwaysSpawn" "MultiplyWithRoot" 0 0 0 0))
                            (Class|AmmoSpawn|InitializePickup _returnvalue_8 _self _array_index _pickuptier)
                            (Utilities|Array|Add (Variables|Default|GetPickups) _returnvalue_8))
                          (:6
                            (bind _returnvalue_9 (Game|SpawnActorfromClass "/FigmentCore/TheKeep/Ammo_Spawn.Ammo_Spawn_C" (Transformation|GetWorldTransform (Variables|Default|GetPlacement7)) "AlwaysSpawn" "MultiplyWithRoot" 0 0 0 0))
                            (Class|AmmoSpawn|InitializePickup _returnvalue_9 _self _array_index _pickuptier)
                            (Utilities|Array|Add (Variables|Default|GetPickups) _returnvalue_9))
                          (:7
                            (bind _returnvalue_10 (Game|SpawnActorfromClass "/FigmentCore/TheKeep/Ammo_Spawn.Ammo_Spawn_C" (Transformation|GetWorldTransform (Variables|Default|GetPlacement8)) "AlwaysSpawn" "MultiplyWithRoot" 0 0 0 0))
                            (Class|AmmoSpawn|InitializePickup _returnvalue_10 _self _array_index _pickuptier)
                            (Utilities|Array|Add (Variables|Default|GetPickups) _returnvalue_10)))))))))
            (elif (Variables|Default|GetWarmup)
              (Variables|Default|SetWarmup false))))))))

(event MissionHelpers|HandleMissionEvent%28v2%29:OnDefenseStopped
  (params (InMissionGuid :type Guid%20Structure) (ObjectiveHandle :type Gameplay%20Tag%20Container%20Structure) (Params :type Fort%20Mission%20Event%20Params%20Object%20Reference) (DO_NOT_USE_THIS_OR_VARIABLES_BELOW :type Boolean) (EventFocus :type Object%20Reference) (EventContent :type Data%20Asset%20Object%20Reference) (EventInstigator :type Actor%20Object%20Reference) (GenericInt :type Integer) (GenericFloat :type Float%20%28single-precision%29) (GenericText :type Text%20%28by%20ref%29) (GameplayTags :type Gameplay%20Tag%20Container%20Structure) (MissionEvent :type Fort%20Mission%20Event%20Structure%20%28by%20ref%29))
  (bind _self self)
  (bind _ammopositions (Variables|Default|GetAmmoPositions))
  (bind _pickups (Variables|Default|GetPickups))
  (bind _pickuptier (Variables|Default|GetPickupTier))
  (Utilities|FlowControl|Sequence
    (:then_0
      (Utilities|FlowControl|Gate true :!label _label
        (:Exit
          (Utilities|FlowControl|Sequence
            (:then_0
              (if (Variables|Default|GetWarmup)
                (Utilities|FlowControl|Delay 0.2)
                (else
                  (Utilities|FlowControl|Delay (Math|Conversions|ToFloat(Integer) (Variables|Default|GetCooldown)))))
              (goto _label))
            (:then_1
              (if (< (Variables|Default|GetAmmoNumber) (Variables|Default|GetMaxPickups))
                (bind (_array_element _array_index) (Utilities|Array|ForEachLoopwithBreak _ammopositions :!label _label_2)
                  (:LoopBody
                    (if _array_element
                      (if (== _array_index 7)
                        _)
                      (else
                        (Utilities|Array|SetArrayElem _ammopositions _array_index "true")
                        (Utilities|FlowControl|Sequence
                          (:then_0
                            (goto _label_2 Break))
                          (:then_1
                            (Variables|Default|SetAmmoNumber (+ (Variables|Default|GetAmmoNumber) 1))
                            (switch int _array_index
                              (:0
                                (bind _returnvalue (Game|SpawnActorfromClass "/FigmentCore/TheKeep/Ammo_Spawn.Ammo_Spawn_C" (Transformation|GetWorldTransform (Variables|Default|GetPlacement%201)) "AlwaysSpawn" "MultiplyWithRoot" 0 0 0 0))
                                (Class|AmmoSpawn|InitializePickup _returnvalue _self _array_index _pickuptier)
                                (Utilities|Array|Add (Variables|Default|GetPickups) _returnvalue))
                              (:1
                                (bind _returnvalue_2 (Game|SpawnActorfromClass "/FigmentCore/TheKeep/Ammo_Spawn.Ammo_Spawn_C" (Transformation|GetWorldTransform (Variables|Default|GetPlacement2)) "AlwaysSpawn" "MultiplyWithRoot" 0 0 0 0))
                                (Class|AmmoSpawn|InitializePickup _returnvalue_2 _self _array_index _pickuptier)
                                (Utilities|Array|Add (Variables|Default|GetPickups) _returnvalue_2))
                              (:2
                                (bind _returnvalue_3 (Game|SpawnActorfromClass "/FigmentCore/TheKeep/Ammo_Spawn.Ammo_Spawn_C" (Transformation|GetWorldTransform (Variables|Default|GetPlacement3)) "AlwaysSpawn" "MultiplyWithRoot" 0 0 0 0))
                                (Class|AmmoSpawn|InitializePickup _returnvalue_3 _self _array_index _pickuptier)
                                (Utilities|Array|Add (Variables|Default|GetPickups) _returnvalue_3))
                              (:3
                                (bind _returnvalue_4 (Game|SpawnActorfromClass "/FigmentCore/TheKeep/Ammo_Spawn.Ammo_Spawn_C" (Transformation|GetWorldTransform (Variables|Default|GetPlacement4)) "AlwaysSpawn" "MultiplyWithRoot" 0 0 0 0))
                                (Class|AmmoSpawn|InitializePickup _returnvalue_4 _self _array_index _pickuptier)
                                (Utilities|Array|Add (Variables|Default|GetPickups) _returnvalue_4))
                              (:4
                                (bind _returnvalue_5 (Game|SpawnActorfromClass "/FigmentCore/TheKeep/Ammo_Spawn.Ammo_Spawn_C" (Transformation|GetWorldTransform (Variables|Default|GetPlacement5)) "AlwaysSpawn" "MultiplyWithRoot" 0 0 0 0))
                                (Class|AmmoSpawn|InitializePickup _returnvalue_5 _self _array_index _pickuptier)
                                (Utilities|Array|Add (Variables|Default|GetPickups) _returnvalue_5))
                              (:5
                                (bind _returnvalue_6 (Game|SpawnActorfromClass "/FigmentCore/TheKeep/Ammo_Spawn.Ammo_Spawn_C" (Transformation|GetWorldTransform (Variables|Default|GetPlacement6)) "AlwaysSpawn" "MultiplyWithRoot" 0 0 0 0))
                                (Class|AmmoSpawn|InitializePickup _returnvalue_6 _self _array_index _pickuptier)
                                (Utilities|Array|Add (Variables|Default|GetPickups) _returnvalue_6))
                              (:6
                                (bind _returnvalue_7 (Game|SpawnActorfromClass "/FigmentCore/TheKeep/Ammo_Spawn.Ammo_Spawn_C" (Transformation|GetWorldTransform (Variables|Default|GetPlacement7)) "AlwaysSpawn" "MultiplyWithRoot" 0 0 0 0))
                                (Class|AmmoSpawn|InitializePickup _returnvalue_7 _self _array_index _pickuptier)
                                (Utilities|Array|Add (Variables|Default|GetPickups) _returnvalue_7))
                              (:7
                                (bind _returnvalue_8 (Game|SpawnActorfromClass "/FigmentCore/TheKeep/Ammo_Spawn.Ammo_Spawn_C" (Transformation|GetWorldTransform (Variables|Default|GetPlacement8)) "AlwaysSpawn" "MultiplyWithRoot" 0 0 0 0))
                                (Class|AmmoSpawn|InitializePickup _returnvalue_8 _self _array_index _pickuptier)
                                (Utilities|Array|Add (Variables|Default|GetPickups) _returnvalue_8)))))))))
                (elif (Variables|Default|GetWarmup)
                  (Variables|Default|SetWarmup false))))))))
    (:then_1
      (for _array_element_2 _pickups
        (Actor|DestroyActor _array_element_2))
      (Utilities|Array|Clear _pickups)
      (Variables|Default|SetAmmoNumber 0))
    (:then_2
      ;[1 1 1] Temp Solution - Remove
      (bind _asfort_player_controller (Utilities|Casting|CastToFortPlayerController (Game|GetPlayerController 0)))
      (Keep|ResetContainer _asfort_player_controller))))

(fn AmmoTaken
  (params (AmmoPos :type Integer) (Reference :type Keep%20Item%20Container%20Object%20Reference))
  (Utilities|Array|SetArrayElem (Variables|Default|GetAmmoPositions) AmmoPos)
  (Utilities|Array|RemoveItem (Variables|Default|GetPickups) Reference)
  (Variables|Default|SetAmmoNumber (- (Variables|Default|GetAmmoNumber) 1)))

(fn BlueprintCanInteract
  (params (InteractingPawn :type Fort%20Pawn%20Object%20Reference) (InteractionBeingAttempted :type EInteractionBeingAttempted%20Enum :default "FirstInteraction") (InteractionType :type TInteractionType%20Enum :default "IT_NoInteraction"))
  (return (> (Keep|GetLootTierOverride) 0)))

(fn ConstructionScript)
